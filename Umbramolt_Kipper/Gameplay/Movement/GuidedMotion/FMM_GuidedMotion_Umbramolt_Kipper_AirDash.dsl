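; Guided-motion movement mode extension for Umbramolt Kipper's air dash, a constant-force dash along the net-synced aim direction.
; On activation it derives strength from dash speed plus a downward-aim bonus, recomputes direction and strength toward the projected endpoint, and caches a per-axis-scaled, clamped finish velocity.
; Override functions return these cached values to guided motion.
; A helper maps downward-trace ground distance to a minimum allowed dot, though nothing shown here calls it.
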(event EventOnExtensionActivate
  (params (RuntimeContext :type MMERuntime%20Context%20Structure%20%28by%20ref%29))
  (bind _initialdashdirection (Variables|Default|GetInitialDashDirection))
  ;[1 1 1] Calculate the initial direction of the dash
  (Variables|Default|SetInitialDashDirection (Math|Vector|GetForwardVector (CharacterMovement|GetNetSyncedBaseAimRotation (Variables|FortMovementMode_BaseExtLogic|GetOwnerMovementComponent))))
  ;[1 1 1] Adjust dash direction and strength
  (bind _strength (Default|MakeFinalStrength _initialdashdirection))
  (bind (_adjusteddirection _adjustedstrength) (Default|AdjustDirectionandStrength (Transformation|GetActorLocation (Variables|FortMovementMode_BaseExtLogic|GetOwnerPawn)) _initialdashdirection _strength))
  (Variables|Default|SetDashDirection _adjusteddirection)
  (Variables|Default|SetDashStrength _adjustedstrength)
  ;[1 1 1] Pre-calculate finish velocity
  (Variables|Default|SetFinishVelocity (Math|Vector|ClampVectorSize (* (Variables|Default|GetInitialDashDirection) (Math|Vector|MakeVector (ScalableFloat|GetValueatLevel (Variables|Default|GetHF_XYVelocityMaintained)) (ScalableFloat|GetValueatLevel (Variables|Default|GetHF_XYVelocityMaintained)) (ScalableFloat|GetValueatLevel (Variables|Default|GetHF_ZVelocityMaintained)))) 0.0 (/ _strength (ScalableFloat|GetValueatLevel (Variables|Config-Shared|GetMaxDuration)))))
  (|CallParentFunction RuntimeContext))

(fn AdjustDirectionAndStrength
  (params (StartLocation :type Vector) (DesiredDirection :type Vector) (Strength :type Float%20%28double-precision%29))
  (return (/ (- (+ (Variables|GetStartLocation) (Utilities|Operators|Multiply DesiredDirection Strength (ScalableFloat|GetValueatLevel (Variables|Config-Shared|GetMaxDuration)))) (Transformation|GetActorLocation (Variables|FortMovementMode_BaseExtLogic|GetOwnerPawn))) (Math|Vector|VectorLength (- (+ (Variables|GetStartLocation) (Utilities|Operators|Multiply DesiredDirection Strength (ScalableFloat|GetValueatLevel (Variables|Config-Shared|GetMaxDuration)))) (Transformation|GetActorLocation (Variables|FortMovementMode_BaseExtLogic|GetOwnerPawn))))) (/ (Math|Vector|VectorLength (- (+ (Variables|GetStartLocation) (Utilities|Operators|Multiply DesiredDirection Strength (ScalableFloat|GetValueatLevel (Variables|Config-Shared|GetMaxDuration)))) (Transformation|GetActorLocation (Variables|FortMovementMode_BaseExtLogic|GetOwnerPawn)))) (ScalableFloat|GetValueatLevel (Variables|Config-Shared|GetMaxDuration)))))

(fn GetDirection_ConstantForce
  (params (GuidedMotionRuntimeData :type Fort%20Movement%20Mode%20Guided%20Motion%20Runtime%20Data%20Object%20Reference))
  (return (Variables|Default|GetDashDirection)))

(fn GetFinishVelocityToSet
  (params (GuidedMotionRuntimeData :type Fort%20Movement%20Mode%20Guided%20Motion%20Runtime%20Data%20Object%20Reference))
  (return (Variables|Default|GetFinishVelocity)))

(fn GetMinAllowedDot
  (bind _ownerpawn (Variables|FortMovementMode_BaseExtLogic|GetOwnerPawn))
  (bind _returnvalue (Transformation|GetActorLocation _ownerpawn))
  (bind _outhit (Collision|LineTraceByChannel (Variables|FortMovementMode_BaseExtLogic|GetOwnerPawn) _returnvalue (Math|Vector|MakeVector (.x _returnvalue) (.y _returnvalue) (- (.z _returnvalue) (ScalableFloat|GetValueatLevel (Variables|Default|GetMaxDistanceAboveGround)))) "TraceTypeQuery10"))
  (return (Math|Float|MapRangeClamped (Math|Vector|Distance(Vector) _returnvalue (Math|Vector|SelectVector (.ImpactPoint (Collision|BreakHitResult _outhit)) (.TraceEnd (Collision|BreakHitResult _outhit)) (Collision|BreakHitResult _outhit))) 0.0 (ScalableFloat|GetValueatLevel (Variables|Default|GetMaxDistanceAboveGround)) (ScalableFloat|GetValueatLevel (Variables|Default|GetMinAllowedDot)) (ScalableFloat|GetValueatLevel (Variables|Default|GetMaxAllowedDot)))))

(fn GetStrength_ConstantForce
  (params (GuidedMotionRuntimeData :type Fort%20Movement%20Mode%20Guided%20Motion%20Runtime%20Data%20Object%20Reference))
  (return (Variables|Default|GetDashStrength)))

(fn MakeFinalStrength
  (params (Direction :type Vector))
  (return (Math|Float|Clamp(Float) (+ (ScalableFloat|GetValueatLevel (Variables|Default|GetHF_DashSpeed)) (Math|Float|MapRangeClamped (Math|Vector|DotProduct Direction "0.000000,0.000000,-1.000000") 0.0 1.0 0.0 (ScalableFloat|GetValueatLevel (Variables|Default|GetHF_DownwardAdditive)))) (* (+ (ScalableFloat|GetValueatLevel (Variables|Default|GetHF_DashSpeed)) (Math|Float|MapRangeClamped (Math|Vector|DotProduct (Variables|Config-ConstantForce|GetDirection) "0.000000,0.000000,-1.000000") 0.0 1.0 0.0 (ScalableFloat|GetValueatLevel (Variables|Default|GetHF_DownwardAdditive)))) 0.95) (* (+ (ScalableFloat|GetValueatLevel (Variables|Default|GetHF_DashSpeed)) (Math|Float|MapRangeClamped (Math|Vector|DotProduct (Variables|Config-ConstantForce|GetDirection) "0.000000,0.000000,-1.000000") 0.0 1.0 0.0 (ScalableFloat|GetValueatLevel (Variables|Default|GetHF_DownwardAdditive)))) 1.05))))
